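# M_Checker is an unlit material in the DirectFX tech-art tools that holds a checker texture lookup.
# MF_ProjectionCoordinates supplies the UVs for the Checker texture sample, with its World Position input left unwired.
# The graph shows no link from the texture sample to a material output, so the checker does not appear to drive the shaded result.
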Material /Game/ArtTools/Tech_Art/DirectFX/Assets/Materials/M_Checker
graph {
 properties {pixelDepthOffsetMode=PDOM_Legacy, shadingModel=MSM_Unlit, usedShadingModels=Unlit}
 nodes [TextureSample(UVs=MaterialFunctionCall!(World Position) {materialFunction=/Game/ArtTools/Tech_Art/DirectFX/Assets/MaterialFunctions/MF_ProjectionCoordinates}.UVCoord) {texture=/Game/FN_PublicTurntable/Turntable_Lowpoly/Content/BlueprintAssets/Checker}]
}
